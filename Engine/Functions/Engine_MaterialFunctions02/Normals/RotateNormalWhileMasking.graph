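# Material function that rotates a world-space normal toward its slope as a greyscale mask increases, useful for forming a lip around thresholded normal-map areas.
# The normalized cross product of the low-poly normal and detail normal gives the rotation axis.
# RotateAboutAxis turns the detail normal around the origin by the mask amount, and its negated offset is added back before renormalizing.
# Both normal inputs default to pixel and vertex world normals.
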MaterialFunction /Engine/Functions/Engine_MaterialFunctions02/Normals/RotateNormalWhileMasking
graph {
 properties {
  description=This will rotate a normal toward it's slope as the greyscale input increases. This is useful for thresholding a normal map and generating a lip around the newly masked areas.
  libraryCategoriesText [Misc,Normals]
 }
 nodes [
  n1 = FunctionInput(Preview=PixelNormalWS()) {
   bUsePreviewValueAsDefault=true
   description="Enter a normal texture transformed into world space. "
   inputName=Hp Normal
   previewValue=<0 0 1 1>
  }
  n2 = Normalize(VectorInput=@n1)
  n3 = CrossProduct(A=FunctionInput(Preview=VertexNormalWS()) {bUsePreviewValueAsDefault=true, inputName=LowPolyNormal, sortPriority=2}, B=@n2) {desc=up vec}
  n4 = FunctionInput() {
   bUsePreviewValueAsDefault=true
   description="Enter a mask value (this will control the number of rotations the normal makes."
   inputName=Grey Scale Mask Amount
   inputType=FunctionInput_Scalar
   sortPriority=1
  }
  n5 = Multiply(A=RotateAboutAxis(NormalizedRotationAxis=Normalize(VectorInput=@n3), RotationAngle=@n4, PivotPoint=Constant(), Position=@n2)) {constB=-1}
  FunctionOutput(None=Normalize(VectorInput=Add(A=@n2, B=@n5))) {outputName=Normal}
 ]
}
