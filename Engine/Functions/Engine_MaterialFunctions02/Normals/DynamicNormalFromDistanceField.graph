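# Material function that generates animated dissolve normals and a mask from three offset distance-field inputs, forming a lip as the field erodes.
# A dissolve alpha, defaulting to a time-driven HeightLerp, thresholds the packed fields, which contrast and power then shape by lip tightness and curvature.
# Normal XY subtracts red and green each from itself as wired, yielding zero before DeriveNormalZ.
# Blue supplies the mask, and thresholded gradients are also output.
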MaterialFunction /Engine/Functions/Engine_MaterialFunctions02/Normals/DynamicNormalFromDistanceField
graph {
 properties {
  bExposeToLibrary=true
  description="Generate animated normals from offset distance fields using this material function. A normal lip will form as the distance field errodes away using a user specified alpha value. Note: multiple distance fields can dynamically be combined together using a \"min\" operation to generate fluid results. If you wish to invert the results invert the distance field before feeding it into the function. "
 }
 nodes [
  n1 = Constant() {r=1}
  n2 = TextureSample(UVs=TextureCoordinate()) {
   samplerType=SAMPLERTYPE_LinearColor
   texture=/Engine/Functions/Engine_MaterialFunctions02/ExampleContent/Textures/OffsetNoiseDistanceFields_ForNormals
  }
  n3 = FunctionInput(Preview=@n2.R) {
   bUsePreviewValueAsDefault=true
   description="Create a texture that is a distance field, offset the red channel x units to the right and the green channel x units upward."
   inputName=Right Shifted Distance Field
   inputType=FunctionInput_Scalar
   sortPriority=-10
  }
  n4 = FunctionInput(Preview=@n2.G) {
   bUsePreviewValueAsDefault=true
   description="Create a texture that is a distance field, offset the red channel x units to the right and the green channel x units upward."
   inputName=Downward Shifted Distance Field
   inputType=FunctionInput_Scalar
   sortPriority=-9
  }
  n5 = FunctionInput(Preview=@n2.B) {
   bUsePreviewValueAsDefault=true
   description="Create a texture that is a distance field, offset the red channel x units to the right and the green channel x units upward."
   inputName=Unshifted distance field
   inputType=FunctionInput_Scalar
   sortPriority=-8
  }
  n6 = MaterialFunctionCall(X=@n3, Y=@n4, Z=@n5) {materialFunction=&MakeFloat3}
  n7 = MaterialFunctionCall(Speed=Constant() {r=-0.25}) {materialFunction=&TimeWithSpeedVariable}
  n8 = OneMinus(None=MaterialFunctionCall() {materialFunction=&LinearGradient}.VGradient)
  n9 = OneMinus(None=FunctionInput(Preview=MaterialFunctionCall(Transition Phase=@n7.Frac Time, Height Texture=@n8) {materialFunction=&HeightLerp}.Alpha) {bUsePreviewValueAsDefault=true, inputName=Dissolve Alpha, inputType=FunctionInput_Scalar})
  n10 = Constant() {r=2}
  n11 = Clamp(None=Add(A=Subtract(A=@n6.Result, B=@n1), B=Multiply(A=@n9, B=@n10))) {desc=height lerp}
  n12 = MaterialFunctionCall(Float3=@n11) {materialFunction=&BreakOutFloat3Components}
  n13 = FunctionInput() {bUsePreviewValueAsDefault=true, inputName=Normal Lip Tightness, inputType=FunctionInput_Scalar, previewValue=<2 0 0 1>}
  n14 = MaterialFunctionCall(
   X=MaterialFunctionCall(In=@n12.R, Contrast=@n13) {materialFunction=&CheapContrast}.Result
   Y=MaterialFunctionCall(In=@n12.G, Contrast=@n13) {materialFunction=&CheapContrast}.Result
   Z=MaterialFunctionCall(In=@n12.B, Contrast=@n13) {materialFunction=&CheapContrast}.Result
  ) {materialFunction=&MakeFloat3}
  n15 = FunctionInput() {bUsePreviewValueAsDefault=true, inputName=Normal Lip Curvature, inputType=FunctionInput_Scalar, previewValue=<6 0 0 1>}
  n16 = MaterialFunctionCall(Float3=OneMinus(None=Power(Base=@n14.Result, Exp=@n15))) {materialFunction=&BreakOutFloat3Components}
  n17 = AppendVector(A=Subtract(A=@n16.R, B=@n16.R), B=Subtract(A=@n16.G, B=@n16.G))
  FunctionOutput(None=DeriveNormalZ(InXY=@n17)) {outputName=Normal}
  FunctionOutput(None=@n17) {
   description=This output can be added with similar outputs before Derive Normal Z is run on the composited value. This is exposed to save instructions in that scenario.
   outputName="Normal XY (Advanced)"
   sortPriority=5
  }
  FunctionOutput(None=@n16.B) {outputName=Mask}
  FunctionOutput(None=@n11) {
   description=This output can be used to generate custom normals or other operations. The blue channel can be used to represent a psuedo thickness value.
   outputName="Thresholded Gradients (Advanced)"
   sortPriority=6
  }
 ]
}
